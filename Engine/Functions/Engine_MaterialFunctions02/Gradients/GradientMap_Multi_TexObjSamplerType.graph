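# Material function that maps a 0-1 greyscale value to a color from one row of a multi-gradient texture, exposed under Misc and Gradient.
# The value becomes U, while V is the center of the row chosen by the zero-based index and gradient count, both rounded up.
# The texture input is sampled as Masks with its own sampler rather than shared wrap, unlike GradientMap_Multi, and outputs RGB and A.
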MaterialFunction /Engine/Functions/Engine_MaterialFunctions02/Gradients/GradientMap_Multi_TexObjSamplerType
graph {
 properties {
  bExposeToLibrary=true
  description="Maps a 0-1 value to a color gradient slice. See the default textures for example content. (This material function differs from GradientMap_Multi by using the texture's sampler type rather than utilizing shared:wrap)"
  libraryCategoriesText [Misc,Gradient]
 }
 nodes [
  n1 = TextureSample() {samplerType=SAMPLERTYPE_Masks, texture=/Engine/Functions/Engine_MaterialFunctions02/ExampleContent/Greyscale}
  n2 = FunctionInput() {bUsePreviewValueAsDefault=true, inputName=Number Of Gradients In the Gradient Map, inputType=FunctionInput_Scalar, previewValue=<3 0 0 1>}
  n3 = Divide(A=Constant() {r=1}, B=Ceil(None=@n2))
  n4 = Multiply(A=@n3) {constB=0.5}
  n5 = FunctionInput() {
   bUsePreviewValueAsDefault=true
   description="The first index is 0 and the last index is (number of gradients in the gradient map) -1."
   inputName="Index (0 Based)"
   inputType=FunctionInput_Scalar
  }
  n6 = Multiply(A=@n3, B=Ceil(None=@n5))
  n7 = TextureSample(
   UVs=AppendVector(
    A=FunctionInput(Preview=@n1.R) {bUsePreviewValueAsDefault=true, inputName=Greyscale Value To Gradient Map, inputType=FunctionInput_Scalar}
    B=Add(A=@n4, B=@n6)
   )
   Tex=FunctionInput(Preview=TextureObject() {texture=/Engine/Functions/Engine_MaterialFunctions02/ExampleContent/Gradient}) {bUsePreviewValueAsDefault=true, inputName=Gradient Texture, inputType=FunctionInput_Texture2D}
  ) {samplerType=SAMPLERTYPE_Masks}
  FunctionOutput(None=@n7.RGB) {outputName=RGB}
  FunctionOutput(None=@n7.A) {outputName=A, sortPriority=1}
 ]
}
