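# Material function in the Distance Fields library that computes an analytic sphere falloff gradient along a ray, defaulting to camera position, camera vector and actor position.
# It projects the sphere center onto the infinite ray line, subtracts the radius from the center-to-line distance, then divides by the falloff.
# Outputs provide clamped and unclamped results, increasing away from the sphere.
# A parameter-driven duplicate chain ending in OneMinus feeds no output.
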MaterialFunction /Engine/Functions/Engine_MaterialFunctions02/Gradients/RayTraceSphereFalloff
graph {
 properties {bExposeToLibrary=true, description=finds the distance to the sphere, libraryCategoriesText [Distance Fields]}
 nodes [
  n1 = Constant()
  n2 = FunctionInput(Preview=CameraVectorWS() {desc=UnitRayDirection}) {bUsePreviewValueAsDefault=true, inputName=Unit Ray Direction, sortPriority=2}
  n3 = FunctionInput(Preview=ActorPositionWS() {desc=sphere center}) {bUsePreviewValueAsDefault=true, inputName=Sphere Center, sortPriority=3}
  n4 = FunctionInput(Preview=CameraPositionWS() {desc=ray origin}) {bUsePreviewValueAsDefault=true, inputName=Ray Origin, sortPriority=4}
  n5 = DotProduct(A=@n2, B=Subtract(A=@n3, B=@n4))
  n6 = Subtract(A=Add(A=Multiply(A=@n5, B=@n2), B=@n4), B=@n3)
  n7 = FunctionInput() {bUsePreviewValueAsDefault=true, inputName=Sphere Radius, inputType=FunctionInput_Scalar, previewValue=<16 0 0 1>}
  n8 = FunctionInput() {bUsePreviewValueAsDefault=true, inputName=Sphere Falloff, inputType=FunctionInput_Scalar, sortPriority=1, previewValue=<16 0 0 1>}
  n9 = Divide(A=Subtract(A=Distance(A=@n1, B=@n6), B=@n7), B=@n8) {constB=10}
  FunctionOutput(None=Clamp(None=@n9)) {outputName=Clamped Falloff}
  FunctionOutput(None=@n9) {outputName=Unclamped Results}
  n10 = CameraPositionWS() {desc=ray origin}
  n11 = CameraVectorWS() {desc=UnitRayDirection}
  n12 = ActorPositionWS() {desc=sphere center}
  n13 = Subtract(A=@n12, B=@n10)
  n14 = Add(A=@n10, B=Multiply(A=DotProduct(A=@n11, B=@n13), B=@n11))
  n15 = Subtract(A=Distance(A=Constant(), B=Subtract(A=@n14, B=@n12)), B=ScalarParameter() {defaultValue=10, parameterName=Sphere Radius})
  n16 = ScalarParameter() {defaultValue=10, parameterName=Sphere Falloff}
  OneMinus(None=Clamp(None=Divide(A=@n15, B=@n16) {constB=10}))
 ]
}
