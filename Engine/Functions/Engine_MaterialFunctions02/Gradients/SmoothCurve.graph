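# Material function in the Misc and Gradient library categories that remaps scalar X along a cubic curve adjusted by two tangent inputs.
# X defaults to the U output of LinearGradient, and both tangents share a default constant.
# It sums linear, squared and cubed terms: T0·X + T0(1−T1)·X² + (T0+T1−2)·X³.
# The curve starts at zero with slope T0, but its value at X=1 varies with the tangents.
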MaterialFunction /Engine/Functions/Engine_MaterialFunctions02/Gradients/SmoothCurve
graph {
 properties {bExposeToLibrary=true, description=Adjust the tangents at x0 and x1 to adjust x's curve., libraryCategoriesText [Misc,Gradient]}
 nodes [
  n1 = FunctionInput(Preview=MaterialFunctionCall() {materialFunction=&LinearGradient}.UGradient) {bUsePreviewValueAsDefault=true, inputName=X, inputType=FunctionInput_Scalar}
  n2 = Constant() {r=2}
  n3 = FunctionInput(Preview=@n2) {bUsePreviewValueAsDefault=true, inputName=Tangent 0, inputType=FunctionInput_Scalar, sortPriority=1}
  n4 = FunctionInput(Preview=@n2) {bUsePreviewValueAsDefault=true, inputName=Tangent 1, inputType=FunctionInput_Scalar, sortPriority=2}
  n5 = Constant() {r=2}
  n6 = Subtract(A=Add(A=@n3, B=@n4), B=@n5)
  n7 = Multiply(A=@n1, B=Multiply(A=@n1, B=Multiply(A=@n1, B=@n6)))
  n8 = Subtract(A=Subtract(A=Constant() {r=3}, B=@n4), B=@n5)
  n9 = Multiply(A=@n1, B=Multiply(A=@n1, B=Multiply(A=@n8, B=@n3)))
  FunctionOutput(None=Add(A=Multiply(A=@n1, B=@n3), B=Add(A=@n7, B=@n9)))
 ]
}
