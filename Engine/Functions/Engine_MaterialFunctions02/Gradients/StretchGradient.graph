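# StretchGradient is a Texturing library material function that scales a 0-1 gradient while keeping both endpoints, stretching it at a chosen center point.
# The scaled gradient is clamped up to the center on one side, and offset by scale minus one then clamped from the center on the other.
# A Ceil-driven lerp joins the halves, leaving a flat band.
# A zero-radius SphereMask outputs that band as a mask.
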MaterialFunction /Engine/Functions/Engine_MaterialFunctions02/Gradients/StretchGradient
graph {
 properties {
  bExposeToLibrary=true
  description=A 0-1 uv gradient is scaled by a desired amount while retaining the uvs 0 and 1 value locations. It does this by stretching out the gradient at a point of the user's choosing.
  libraryCategoriesText [Texturing]
 }
 nodes [
  n1 = FunctionInput(Preview=MaterialFunctionCall(Float2=TextureCoordinate()) {materialFunction=&BreakOutFloat2Components}.R) {bUsePreviewValueAsDefault=true, inputName=0-1 Gradient, inputType=FunctionInput_Scalar}
  n2 = FunctionInput() {bUsePreviewValueAsDefault=true, inputName=Scale Factor, inputType=FunctionInput_Scalar, previewValue=<2 0 0 1>}
  n3 = Multiply(A=@n1, B=@n2) {constB=2}
  n4 = FunctionInput() {bUsePreviewValueAsDefault=true, inputName=Center Point, inputType=FunctionInput_Scalar, previewValue=<0.5 0 0 1>}
  n5 = Subtract(A=@n2)
  n6 = LinearInterpolate(A=Clamp(None=@n3, Max=@n4), B=Clamp(None=Subtract(A=@n3, B=@n5), Min=@n4), Alpha=Ceil(None=Subtract(A=@n1, B=@n4) {constB=0.5}))
  FunctionOutput(None=@n6) {outputName=Resulting Gradient}
  FunctionOutput(None=SphereMask(A=@n6, B=@n4) {attenuationRadius=0}) {description="This returns a white value for the pixels that have been stretched. ", outputName=Clamped area, sortPriority=1}
 ]
}
